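# Physics asset of cape collider bodies for the Llama Rider female medium player cosmetic, variant 04.
# Capsules cover the head, clavicles, pelvis, spine and feet, plus the llama's back, rear legs and tail; several middle spine bones add a second rotated capsule.
# Arm, neck, thigh and calf bodies hold no shapes.
# It has no constraints, so it supplies collision volumes rather than a jointed ragdoll.
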PhysicsAsset /BRCosmetics/Characters/Player/Female/Medium/Bodies/F_MED_LlamaRider/Meshes/F_MED_LlamaRider_04_CapeColliders
bodies {
 calf_l []
 calf_r []
 clavicle_l [{shapeType=Capsule, length=7.6964, radius=6.3853, center=<4.0147 -0.8354 -1.7419>, rotation=<rot -77.5566 177.3915 7.7314>}]
 clavicle_r [{shapeType=Capsule, length=7.6964, radius=8.2957, center=<-4.0147 0.8354 1.7419>, rotation=<rot -77.5565 177.3915 -172.2685>}]
 foot_l [{shapeType=Capsule, length=12.5982, radius=6.5127, center=<-2.2072 6.0534 -0.3367>, rotation=<rot 0.0388 1.4701 -89.3982>}]
 foot_r [{shapeType=Capsule, length=12.5982, radius=6.5127, center=<2.2072 -6.0534 0.3367>, rotation=<rot 0.0388 1.4701 90.6018>}]
 head [{shapeType=Capsule, length=0, radius=18, center=<9.0697 0.0746 0>, rotation=<rot 0 -90 0>}]
 llama_back_l [{
   shapeType=Capsule
   length=10.75
   radius=22
   shapeName=llama_back_l_capsule
   center=<-14.899 -2.6269 4.0653>
   rotation=<rot -90 -91.6162 -178.3848>
  }]
 llama_back_r [{
   shapeType=Capsule
   length=10.75
   radius=22
   shapeName=llama_back_r_capsule
   center=<16.3468 1.4526 -3.9294>
   rotation=<rot -90 -177.6141 -92.3869>
  }]
 llama_leg_back_l [{
   shapeType=Capsule
   length=21.5
   radius=8.5
   shapeName=llama_leg_back_l_capsule
   center=<14.8723 0.0001 -3.1472>
   rotation=<rot -90 180 -179.9997>
  }]
 llama_leg_back_r [{
   shapeType=Capsule
   length=21.5
   radius=8.5
   shapeName=llama_leg_back_r_capsule
   center=<-16.4541 -0.0001 2.3264>
   rotation=<rot -90 90 -89.9997>
  }]
 llama_tail [{shapeType=Capsule, length=6, radius=10.75, shapeName=llama_tail_capsule, center=<-0.0003 -13.0541 0>, rotation=<rot -90 14.0362 -104.0374>}]
 lowerarm_l []
 lowerarm_r []
 neck_02 []
 pelvis [{shapeType=Capsule, length=9.4028, radius=10.6596, center=<-1.9544 -2.2029 -0.1981>}]
 spine_01 [{shapeType=Capsule, length=5.0171, radius=13.7354, center=<2.3275 -3.2584 0>}]
 spine_02 [
  {shapeType=Capsule, length=3.2984, radius=8.4854, center=<1.179 2.2974 0>}
  {shapeType=Capsule, length=20.6875, radius=9.3125, center=<-1.8317 2.606 0>, rotation=<rot 90 0 0>}
 ]
 spine_03 [
  {shapeType=Capsule, length=0, radius=13.9854, center=<0.9296 0.668 0>}
  {shapeType=Capsule, length=20.7188, radius=9.2812, center=<-1.2326 2.6059 0>, rotation=<rot 90 0 0>}
 ]
 spine_04 [
  {shapeType=Capsule, length=5.8296, radius=10.3916, center=<5.5182 1.4112 0>}
  {shapeType=Capsule, length=10.0312, radius=9.2812, center=<-1.2326 1.308 0>, rotation=<rot 90 0 0>}
 ]
 spine_05 [{shapeType=Capsule, length=10.9859, radius=10.0791, center=<5.6377 0.3227 0>}]
 thigh_l []
 thigh_r []
 upperarm_l []
 upperarm_r []
}
constraints []
