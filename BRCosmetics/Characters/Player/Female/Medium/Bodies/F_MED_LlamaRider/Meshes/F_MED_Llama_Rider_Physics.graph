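# Physics asset for the female medium Llama Rider cosmetic body, covering the llama costume, rider legs and small dangling accessories.
# Simple box, capsule and sphere bodies on the root, pelvis, spine_05, arms, calves and weapon_r are linked by fully free angular constraints.
# Limited joints hang llama quarters, rider legs and tail from the pelvis, with llama legs and tail tips chained below.
# Grass, laces, bandana and dogtags attach to calves, right arm and spine.
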PhysicsAsset /BRCosmetics/Characters/Player/Female/Medium/Bodies/F_MED_LlamaRider/Meshes/F_MED_Llama_Rider_Physics
bodies {
 arm_bandana [{shapeType=Capsule, length=3.7031, radius=2.1775, center=<-0.1773 4.2949 0.0457>, rotation=<rot -61.4493 165.1061 109.4366>}]
 calf_l [{shapeType=Capsule, length=33.6128, radius=7.4511, center=<-26.0826 0.4678 -0.4726>, rotation=<rot 0 86.9711 -88.9827>}]
 calf_r [{shapeType=Capsule, length=32.2592, radius=5.0857, center=<27.9281 0.8371 0.8124>, rotation=<rot 0 93.133 -89.1245>}]
 clavicle_l [{shapeType=Capsule, length=5, radius=5}]
 clavicle_r [{shapeType=Capsule, length=5, radius=5}]
 dyn_dogtags [{shapeType=Box, extentX=7.5, extentY=1.5754, extentZ=7.5, center=<4.7363 -0.1612 -0.8432>}]
 dyn_laces_r [{shapeType=Capsule, length=4.8534, radius=1.2049, center=<0 -0.0418 -3.6705>, rotation=<rot 0 -90 0>}]
 grass_foot1_l [{shapeType=Box, extentX=3.4978, extentY=3.5463, extentZ=12.6546, center=<5.954 1.8267 -0.4494>, rotation=<rot -25.0058 99.7937 -88.6085>}]
 grass_foot1_r [{shapeType=Box, extentX=3.4978, extentY=3.5463, extentZ=12.6546, center=<3.3239 -0.127 -0.657>, rotation=<rot -25.0058 -80.2058 -88.6085>}]
 grass_foot2_l [{shapeType=Box, extentX=3.4978, extentY=3.5463, extentZ=12.6546, center=<5.954 1.8267 -0.4494>, rotation=<rot -25.0058 99.7937 -88.6085>}]
 grass_foot2_r [{shapeType=Box, extentX=3.4978, extentY=3.5463, extentZ=12.6546, center=<5.3343 -0.5959 1.0624>, rotation=<rot -25.0057 -80.205 -110.6604>}]
 grass_foot3_l [{shapeType=Box, extentX=3.4978, extentY=3.5463, extentZ=12.6546, center=<1.7801 1.0621 -0.5427>, rotation=<rot -21.9431 72.6606 -86.6151>}]
 grass_foot3_r [{shapeType=Box, extentX=3.4978, extentY=3.5463, extentZ=12.6546, center=<4.0856 0.7583 -1.0779>, rotation=<rot -24.2606 -86.2186 -79.7936>}]
 grass_lowerleg_l [{shapeType=Box, extentX=3.4978, extentY=3.5463, extentZ=12.6546, center=<5.954 1.8267 -0.4494>, rotation=<rot -25.0058 99.7937 -88.6085>}]
 grass_lowerleg_r [{shapeType=Box, extentX=3.4978, extentY=3.5463, extentZ=12.6546, center=<5.954 1.8267 -0.4494>, rotation=<rot -25.0058 99.7937 -88.6085>}]
 grass_toe_l [{shapeType=Box, extentX=7.8695, extentY=9.5508, extentZ=11.1002, center=<-0.9052 -3.9408 -0.4687>, rotation=<rot -1.4125 179.149 -94.2554>}]
 grass_toe_r [{shapeType=Box, extentX=7.8695, extentY=9.5508, extentZ=11.1002, center=<0.7674 -4.3187 -1.0835>, rotation=<rot -1.4125 -0.851 -94.2554>}]
 llama_back_l [{
   shapeType=Box
   extentX=11.1732
   extentY=26.9555
   extentZ=35.4591
   center=<-12.3393 -3.3204 -3.9773>
   rotation=<rot -33.4173 -63.7827 74.8245>
  }]
 llama_back_r [{shapeType=Box, extentX=11.1732, extentY=26.9555, extentZ=35.4591, center=<12.3393 3.3204 3.9773>, rotation=<rot -33.4173 -63.7827 -105.1755>}]
 llama_front_l [{shapeType=Box, extentX=39.2206, extentY=23.1513, extentZ=11.1913, center=<-13.4024 15.9721 0>, rotation=<rot 0 -50.0001 -90>}]
 llama_front_r [{shapeType=Box, extentX=39.2206, extentY=23.1513, extentZ=11.1913, center=<13.4024 -15.9721 0>, rotation=<rot 0 -50.0001 89.9999>}]
 llama_leg_back_l [{shapeType=Capsule, length=16.8407, radius=5.8954, center=<12.7458 0.1565 -0.3659>, rotation=<rot 0 90.9288 -89.7511>}]
 llama_leg_back_r [{shapeType=Capsule, length=15, radius=5.8954, center=<-16.3556 -0.8614 2.7598>, rotation=<rot 0 90 90>}]
 llama_leg_front_l [{shapeType=Capsule, length=20.3979, radius=3.8, center=<15.1713 0.2961 0.4869>, rotation=<rot 0 90 -90>}]
 llama_leg_front_r [{shapeType=Capsule, length=18, radius=3.8, center=<-15.1501 0.5129 -0.7408>, rotation=<rot 0 90 -90>}]
 llama_tail [{shapeType=Capsule, length=1.991, radius=3.7229, center=<-0.3342 -4.3802 -0.0523>, rotation=<rot 0 4.1654 -83.5121>}]
 llama_tail1 [{shapeType=Sphere, radius=8.2757, center=<-1.5355 -0.7244 1.1677>}]
 llama_tail2 [{shapeType=Sphere, radius=8.2757, center=<-1.5355 -0.7244 1.1677>}]
 llama_tail3 [{shapeType=Sphere, radius=8.2757, center=<-1.5355 -0.7244 1.1677>}]
 lowerarm_l [{shapeType=Capsule, length=24.75, radius=4.25, center=<14.3537 0 -0.0001>, rotation=<rot -90 0.0396 -0.0396>}]
 lowerarm_r [{shapeType=Capsule, length=24.75, radius=4.25, center=<-14.3537 0 0.0001>, rotation=<rot -90 0.0396 -0.0396>}]
 pelvis [{shapeType=Box, extentX=18.9679, extentY=7.5, extentZ=61, center=<26.9952 -3.0911 0>}]
 rider_leg_l [{
   shapeType=Capsule
   length=29.2804
   radius=10.1942
   shapeName=rider_leg_l_capsule
   center=<-12.4439 23.3403 0.7689>
   rotation=<rot 0 -129.4465 -88.1588>
  }]
 rider_reg_r [{
   shapeType=Capsule
   length=29.2804
   radius=10.1942
   shapeName=rider_reg_r_capsule
   center=<12.4436 -23.3403 -0.7689>
   rotation=<rot 0 -129.4465 -88.1588>
  }]
 root [{shapeType=Box, extentX=117.6319, extentY=117.6319, extentZ=13.9072, center=<0 1.9054 -6.0386>, rotation=<rot 0 -90 0>}]
 spine_05 [{shapeType=Box, extentX=13.4947, extentY=2.0268, extentZ=34.077, center=<4.4989 5.382 0>, rotation=<rot 0 -38.8399 0>}]
 upperarm_l [{shapeType=Capsule, length=24.75, radius=4.25, center=<14.3537 0 -0.0001>, rotation=<rot -90 -0.0264 0.0264>}]
 upperarm_r [{shapeType=Capsule, length=24.75, radius=4.25, center=<-14.3537 0 0.0001>, rotation=<rot -90 0.0396 -0.0396>}]
 weapon_r [{shapeType=Capsule, length=17, radius=4.25, center=<21.4378 0.0009 -0.0011>, rotation=<rot -90 -7.125 27.1283>}]
}
constraints [
 {
  bone1Name=lowerarm_r
  bone2Name=upperarm_r
  swing1LimitDegrees=45
  swing1Motion=Free
  swing2LimitDegrees=45
  swing2Motion=Free
  twistLimitDegrees=45
  twistMotion=Free
 }
 {
  bone1Name=upperarm_r
  bone2Name=clavicle_r
  swing1LimitDegrees=45
  swing1Motion=Free
  swing2LimitDegrees=45
  swing2Motion=Free
  twistLimitDegrees=45
  twistMotion=Free
 }
 {
  bone1Name=clavicle_r
  bone2Name=spine_05
  swing1LimitDegrees=45
  swing1Motion=Free
  swing2LimitDegrees=45
  swing2Motion=Free
  twistLimitDegrees=45
  twistMotion=Free
 }
 {
  bone1Name=weapon_r
  bone2Name=lowerarm_r
  swing1LimitDegrees=45
  swing1Motion=Free
  swing2LimitDegrees=45
  swing2Motion=Free
  twistLimitDegrees=45
  twistMotion=Free
 }
 {
  bone1Name=clavicle_l
  bone2Name=spine_05
  swing1LimitDegrees=45
  swing1Motion=Free
  swing2LimitDegrees=45
  swing2Motion=Free
  twistLimitDegrees=45
  twistMotion=Free
 }
 {
  bone1Name=upperarm_l
  bone2Name=clavicle_l
  swing1LimitDegrees=45
  swing1Motion=Free
  swing2LimitDegrees=45
  swing2Motion=Free
  twistLimitDegrees=45
  twistMotion=Free
 }
 {
  bone1Name=lowerarm_l
  bone2Name=upperarm_l
  swing1LimitDegrees=45
  swing1Motion=Free
  swing2LimitDegrees=45
  swing2Motion=Free
  twistLimitDegrees=45
  twistMotion=Free
 }
 {
  bone1Name=pelvis
  bone2Name=root
  swing1LimitDegrees=45
  swing1Motion=Free
  swing2LimitDegrees=45
  swing2Motion=Free
  twistLimitDegrees=45
  twistMotion=Free
 }
 {bone1Name=llama_leg_back_l, bone2Name=llama_back_l, swing1LimitDegrees=45, swing2LimitDegrees=45, twistLimitDegrees=5}
 {bone1Name=llama_leg_back_r, bone2Name=llama_back_r, swing1LimitDegrees=45, swing2LimitDegrees=45, twistLimitDegrees=5}
 {
  bone1Name=calf_l
  bone2Name=pelvis
  swing1LimitDegrees=45
  swing1Motion=Free
  swing2LimitDegrees=45
  swing2Motion=Free
  twistLimitDegrees=45
  twistMotion=Free
 }
 {
  bone1Name=calf_r
  bone2Name=pelvis
  swing1LimitDegrees=45
  swing1Motion=Free
  swing2LimitDegrees=45
  swing2Motion=Free
  twistLimitDegrees=45
  twistMotion=Free
 }
 {bone1Name=llama_back_l, bone2Name=pelvis, swing1LimitDegrees=22, swing2LimitDegrees=22, twistLimitDegrees=1}
 {bone1Name=rider_reg_r, bone2Name=pelvis, swing1LimitDegrees=20, swing2LimitDegrees=20, twistLimitDegrees=7}
 {bone1Name=rider_leg_l, bone2Name=pelvis, swing1LimitDegrees=20, swing2LimitDegrees=20, twistLimitDegrees=7}
 {bone1Name=llama_front_r, bone2Name=pelvis, swing1LimitDegrees=25, swing2LimitDegrees=25, twistLimitDegrees=12}
 {bone1Name=grass_toe_l, bone2Name=calf_l, swing1LimitDegrees=10, swing2LimitDegrees=10, twistLimitDegrees=1}
 {bone1Name=grass_toe_r, bone2Name=calf_r, swing1LimitDegrees=10, swing2LimitDegrees=10, twistLimitDegrees=1}
 {bone1Name=grass_foot2_l, bone2Name=calf_l, swing1LimitDegrees=10, swing2LimitDegrees=10, twistLimitDegrees=1}
 {bone1Name=grass_foot1_l, bone2Name=calf_l, swing1LimitDegrees=10, swing2LimitDegrees=10, twistLimitDegrees=1}
 {bone1Name=grass_foot3_l, bone2Name=calf_l, swing1LimitDegrees=10, swing2LimitDegrees=10, twistLimitDegrees=1}
 {bone1Name=grass_foot3_r, bone2Name=calf_r, swing1LimitDegrees=10, swing2LimitDegrees=10, twistLimitDegrees=1}
 {bone1Name=grass_foot1_r, bone2Name=calf_r, swing1LimitDegrees=10, swing2LimitDegrees=10, twistLimitDegrees=1}
 {bone1Name=grass_foot2_r, bone2Name=calf_r, swing1LimitDegrees=10, swing2LimitDegrees=10, twistLimitDegrees=1}
 {bone1Name=grass_lowerleg_r, bone2Name=calf_r, swing1LimitDegrees=10, swing2LimitDegrees=10, twistLimitDegrees=1}
 {bone1Name=grass_lowerleg_l, bone2Name=calf_l, swing1LimitDegrees=10, swing2LimitDegrees=10, twistLimitDegrees=1}
 {bone1Name=arm_bandana, bone2Name=upperarm_r, swing1LimitDegrees=20, swing2LimitDegrees=20, twistLimitDegrees=1}
 {bone1Name=dyn_laces_r, bone2Name=calf_r, swing1LimitDegrees=1, swing2LimitDegrees=1, twistLimitDegrees=45}
 {bone1Name=dyn_dogtags, bone2Name=spine_05, swing1LimitDegrees=1, swing2LimitDegrees=1, twistLimitDegrees=45}
 {
  bone1Name=spine_05
  bone2Name=pelvis
  swing1LimitDegrees=45
  swing1Motion=Free
  swing2LimitDegrees=45
  swing2Motion=Free
  twistLimitDegrees=45
  twistMotion=Free
 }
 {bone1Name=llama_back_r, bone2Name=pelvis, swing1LimitDegrees=20, swing2LimitDegrees=20, twistLimitDegrees=7}
 {bone1Name=llama_front_l, bone2Name=pelvis, swing1LimitDegrees=20, swing2LimitDegrees=20, twistLimitDegrees=4}
 {bone1Name=llama_tail, bone2Name=pelvis, swing1LimitDegrees=45, swing2LimitDegrees=45, twistLimitDegrees=45}
 {bone1Name=llama_tail2, bone2Name=llama_tail, swing1LimitDegrees=20, swing2LimitDegrees=20, twistLimitDegrees=5}
 {bone1Name=llama_tail3, bone2Name=llama_tail, swing1LimitDegrees=20, swing2LimitDegrees=20, twistLimitDegrees=5}
 {bone1Name=llama_tail1, bone2Name=llama_tail, swing1LimitDegrees=20, swing2LimitDegrees=20, twistLimitDegrees=5}
 {bone1Name=llama_leg_front_l, bone2Name=llama_front_l, swing1LimitDegrees=45, swing2LimitDegrees=45, twistLimitDegrees=5}
 {bone1Name=llama_leg_front_r, bone2Name=llama_front_r, swing1LimitDegrees=45, swing2LimitDegrees=60, twistLimitDegrees=5}
]
