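# Post-process style AnimBP for the female Llama Rider outfit, adding secondary motion to the pose copied from the attached parent mesh.
# A targeting-driven additive head pose feeds standard and crouch RigidBody sims, then neck/head AnimDynamics and bandana/lace constraints.
# The event graph computes Perlin-noise skydive lift, halved under parachute, and disables head physics during the llama emote.
# Mannequin bone-scaling and leg-adjustment chains sit unconnected to the output.
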AnimBlueprint /BRCosmetics/Characters/Player/Female/Medium/Bodies/F_MED_LlamaRider/Meshes/F_MED_LlamaRider_AnimBP
target_skeleton=&/F_MED_LlamaRider_Skeleton
graphs [
 AnimGraph {nodes [
   n1 = "Transform (Modify) Bone - Bone: llama_neck1"(
    ComponentPose=Local To Component(LocalPose=Copy Pose From Mesh() {bUseAttachedParent=true})
    Alpha=1
    Translation="0,0,5"
    Rotation="15,0,-15"
    Scale="0.7,0.7,0.7"
   ) {
    rotationMode=BMM_Additive
    rotationSpace=BCS_BoneSpace
    scaleMode=BMM_Replace
    scaleSpace=BCS_BoneSpace
    translationMode=BMM_Additive
    translationSpace=BCS_BoneSpace
    alphaBoolBlend {blendOption=Linear}
    boneToModify {boneName=llama_neck1}
    rotation {pitch=15, roll=-15}
    scale=<0.7 0.7 0.7>
    translation {z=5}
   }
   n2 = "Transform (Modify) Bone - Bone: rider_leg_l"(ComponentPose=@n1, Alpha=1, Translation="0,0,0", Rotation="0,0,0", Scale="0.75,0.75,0.75") {
    scaleMode=BMM_Replace
    scaleSpace=BCS_BoneSpace
    alphaBoolBlend {blendOption=Linear}
    boneToModify {boneName=rider_leg_l}
    scale=<0.75 0.75 0.75>
   }
   n3 = "Transform (Modify) Bone - Bone: rider_reg_r"(ComponentPose=@n2, Alpha=1, Translation="0,0,0", Rotation="0,0,0", Scale="0.75,0.75,0.75") {
    scaleMode=BMM_Replace
    scaleSpace=BCS_BoneSpace
    alphaBoolBlend {blendOption=Linear}
    boneToModify {boneName=rider_reg_r}
    scale=<0.75 0.75 0.75>
   }
   n4 = "Transform (Modify) Bone - Bone: llama_front_l"(ComponentPose=@n3, Alpha=1, Translation="0,-5,0", Rotation="0,10,0", Scale="0.5,1,1") {
    rotationMode=BMM_Additive
    rotationSpace=BCS_BoneSpace
    scaleMode=BMM_Replace
    scaleSpace=BCS_BoneSpace
    translationMode=BMM_Additive
    translationSpace=BCS_BoneSpace
    alphaBoolBlend {blendOption=Linear}
    boneToModify {boneName=llama_front_l}
    rotation {yaw=10}
    scale {x=0.5}
    translation {y=-5}
   }
   n5 = "Transform (Modify) Bone - Bone: llama_front_r"(ComponentPose=@n4, Alpha=1, Translation="0,0,0", Rotation="0,0,0", Scale="0.5,1,1") {scaleMode=BMM_Replace, scaleSpace=BCS_BoneSpace, alphaBoolBlend {blendOption=Linear}, boneToModify {boneName=llama_front_r}, scale {x=0.5}}
   n6 = "Transform (Modify) Bone - Bone: spine_01"(ComponentPose=@n5, Alpha=1, Translation="0,0,0", Rotation="0,0,0", Scale="1.35,1.35,1.35") {scaleMode=BMM_Replace, alphaBoolBlend {blendOption=Linear}, boneToModify {boneName=spine_01}, scale=<1.35 1.35 1.35>}
   n7 = "Transform (Modify) Bone - Bone: pelvis"(ComponentPose=@n6, Alpha=1, Translation="0,0,0", Rotation="0,0,0", Scale="0.75,0.75,0.75") {scaleMode=BMM_Replace, alphaBoolBlend {blendOption=Linear}, boneToModify {boneName=pelvis}, scale=<0.75 0.75 0.75>}
   Blend Poses by bool(BlendPose_0=Component To Local(ComponentPose=@n7), bActiveValue=Get isMannequin() -> (isMannequin:bool)) {blendType=Linear}
   n8 = "Transform (Modify) Bone - Bone: llama_leg_back_l"(Alpha=1, Translation="0,0,0", Rotation="0,0,0", Scale="1,1,1") {rotationMode=BMM_Additive, rotationSpace=BCS_BoneSpace, alphaBoolBlend {blendOption=Linear}, boneToModify {boneName=llama_leg_back_l}}
   n9 = "Transform (Modify) Bone - Bone: llama_leg_back_r"(ComponentPose=@n8, Alpha=1, Translation="0,0,0", Rotation="0,0,0", Scale="1,1,1") {rotationMode=BMM_Additive, rotationSpace=BCS_BoneSpace, alphaBoolBlend {blendOption=Linear}, boneToModify {boneName=llama_leg_back_r}}
   n10 = "Transform (Modify) Bone - Bone: llama_leg_front_l"(ComponentPose=@n9, Alpha=1, Translation="0,0,0", Rotation="0,0,0", Scale="1,1,1") {rotationMode=BMM_Additive, rotationSpace=BCS_BoneSpace, alphaBoolBlend {blendOption=Linear}, boneToModify {boneName=llama_leg_front_l}}
   n11 = "Transform (Modify) Bone - Bone: llama_leg_front_r"(ComponentPose=@n10, Alpha=1, Translation="0,0,0", Rotation="0,0,0", Scale="1,1,1") {rotationMode=BMM_Additive, rotationSpace=BCS_BoneSpace, alphaBoolBlend {blendOption=Linear}, boneToModify {boneName=llama_leg_front_r}}
   n12 = "Transform (Modify) Bone - Bone: llama_back_l"(ComponentPose=@n11, Alpha=0.7, Translation="0,0,0", Rotation="0,0,0", Scale="1,1,1") {rotationMode=BMM_Additive, rotationSpace=BCS_BoneSpace, alphaBoolBlend {blendOption=Linear}, boneToModify {boneName=llama_back_l}}
   "Transform (Modify) Bone - Bone: llama_back_r"(ComponentPose=@n12, Alpha=0.7, Translation="0,0,0", Rotation="0,0,0", Scale="1,1,1") {rotationMode=BMM_Additive, rotationSpace=BCS_BoneSpace, alphaBoolBlend {blendOption=Linear}, boneToModify {boneName=llama_back_r}}
   n13 = RigidBody(
    ComponentPose=Local To Component(LocalPose=Use cached pose 'PreSim'())
    ExternalForce=Get FallingAnimDynUpForce() -> (FallingAnimDynUpForce:Vector)
   ) {
    bClampLinearTranslationLimitToRefPose=true
    bOverrideWorldGravity=true
    overridePhysicsAsset=&/F_MED_Llama_Rider_crouch_Physics
    simulationSpace=BaseBoneSpace
    alphaBoolBlend {blendOption=Linear}
    baseBoneRef {boneName=root}
    componentAppliedLinearAccClamp=<1000 1000 1000>
    componentLinearAccScale=<1 1 1.25>
    componentLinearVelScale=<0.5 0.5 0.25>
    overrideWorldGravity {z=-980}
   }
   n14 = Component To Local(ComponentPose=@n13)
   n15 = RigidBody(ComponentPose=Local To Component(LocalPose=Use cached pose 'PreSim'()), ExternalForce=Get FallingAnimDynUpForce()) {
    bClampLinearTranslationLimitToRefPose=true
    bOverrideWorldGravity=true
    overridePhysicsAsset=&/F_MED_Llama_Rider_Physics
    simulationSpace=BaseBoneSpace
    alphaBoolBlend {blendOption=Linear}
    baseBoneRef {boneName=root}
    componentAppliedLinearAccClamp=<1000 1000 1000>
    componentLinearAccScale=<1 1 1.25>
    componentLinearVelScale=<0.5 0.5 0.25>
    overrideWorldGravity {z=-980}
   }
   n16 = Component To Local(ComponentPose=@n15)
   n17 = Get LlamaHeadPhysAlpha() -> (LlamaHeadPhysAlpha:float)
   n18 = Anim Dynamics - Bone: llama_neck1(
    ComponentPose=Local To Component(LocalPose=Blend Poses by bool(BlendPose_0=@n14, BlendPose_1=@n16) {blendType=Linear})
    Alpha=@n17
    ExternalForce="0,0,0"
   ) {
    bGravityOverrideInSimSpace=true
    gravityScale=0.3
    alphaBoolBlend {blendOption=Linear}
    boundBone {boneName=llama_neck1}
    physicsBodyDefinitions [{
      sphereCollisionRadius=0
      boundBone {boneName=llama_neck1}
      boxExtents=<10 20 20>
      constraintSetup {angularLimitsMax=<2 10 10>, angularLimitsMin=<-2 -10 -10>, angularTarget=<0 0 1>}
      localJointOffset=<-15 0 0>
     }]
   }
   n19 = Anim Dynamics - Bone: llama_head(ComponentPose=@n18, Alpha=@n17, ExternalForce="0,0,0") {
    angularSpringConstant=100
    bAngularSpring=true
    bGravityOverrideInSimSpace=true
    gravityScale=0.3
    alphaBoolBlend {blendOption=Linear}
    boundBone {boneName=llama_head}
    physicsBodyDefinitions [{
      sphereCollisionRadius=0
      boundBone {boneName=llama_head}
      boxExtents=<10 40 40>
      constraintSetup {angularLimitsMax=<2 1 20>, angularLimitsMin=<-2 -1 -20>, angularTarget=<0 0 1>}
      localJointOffset=<-7 0 0>
     }]
   }
   n20 = "Constraint - arm_bandana_const (PPP)"!(ComponentPose=@n19, Alpha=1, ConstraintWeights_0=0.5635, ConstraintWeights_1:float=0.21, ConstraintWeights_2:float=0.2265) -> (Pose) {
    alphaBoolBlend {blendOption=Linear}
    boneToModify {boneName=arm_bandana_const}
    constraintSetup [{targetBone {boneName=upperarm_twist_01_r}}, {targetBone {boneName=upperarm_r}}, {targetBone {boneName=upperarm_twist_02_r}}]
    constraintWeights [1,1,1]
   }
   n21 = "Constraint - dyn_laces_r_const (PP)"!(ComponentPose=@n20, Alpha=1, ConstraintWeights_0=0.8525, ConstraintWeights_1:float=0.1475) -> (Pose) {
    alphaBoolBlend {blendOption=Linear}
    boneToModify {boneName=dyn_laces_r_const}
    constraintSetup [{targetBone {boneName=calf_twist_02_r}}, {targetBone {boneName=calf_r}}]
    constraintWeights [1,1]
   }
   Output Pose(Result=Component To Local(ComponentPose=@n21)) {layerGroup=None}
   n22 = F_MED_Llama_Rider_Anim_Pose : head_left!(PoseWeight:float=1) -> (Pose:Pose Link Structure) {
    is=title AnimGraphNode_PoseByName
    poseAsset=/BRCosmetics/Characters/Player/Female/Medium/Bodies/F_MED_LlamaRider/Meshes/PoseAsset/F_MED_Llama_Rider_Anim_Pose
    poseName=head_left
   }
   n23 = Apply Additive!(Base=Copy Pose From Mesh() {bUseAttachedParent=true}, Additive=@n22, bAlphaBoolEnabled:bool=Get bIsTargeting()) -> (Pose) {
    alphaInputType=Bool
    alphaBoolBlend {blendInTime=0.2, blendOption=Custom, customCurve=/Game/Characters/Player/Male/Medium/Bodies/M_MED_LlamaRider/Meshes/llama_head_driver_curve}
   }
   Save cached pose 'PreSim'(Pose=@n23) {cacheName=PreSim}
   Get CurrentDiveJiggle() -> (CurrentDiveJiggle:Rotator)
  ]}
 EventGraph {nodes [
   Sequence()
   > n1 = Cast To FortPlayerMannequin!(execute, Object=GetOwningActor()) -> (then, CastFailed, AsFort Player Mannequin:Fort Player Mannequin Object Reference) {is=title K2Node_DynamicCast}
   > Set isMannequin(isMannequin:bool=IsValid(Object=@n1)) -> (Output_Get:bool)
   Set isMannequin(execute=@n1.CastFailed, isMannequin=false)
   Reroute Node()
   n2 = BreakVector(InVec=Get PawnLastLocation() -> (PawnLastLocation:Vector))
   > Set current_actor_z(current_actor_z:double=@n2.Z) -> (Output_Get:double)
   n3 = Make Perlin Noise and Remap!(Value:float=float + float(A=@n2.Z, B=20) {is=title K2Node_CommutativeAssociativeBinaryOperator}, RangeOutMin:float=-30, RangeOutMax:float=30) -> (ReturnValue:float)
   n4 = Make Perlin Noise and Remap(Value=@n2.Z, RangeOutMin=-30, RangeOutMax=30)
   > Set current_dive_jiggle(current_dive_jiggle:Rotator=RInterpTo!(
     Current:Rotator=Get current_dive_jiggle() -> (current_dive_jiggle:Rotator)
     Target:Rotator=MakeRotator(Pitch=@n4, Yaw=@n3)
     DeltaTime:float=0
     InterpSpeed:float=10
    ) -> (ReturnValue:Rotator)) -> (Output_Get:Rotator)
   n5 = SelectFloat(
    A=Make Perlin Noise and Remap(Value=Get current_actor_z() -> (current_actor_z:double), RangeOutMin=0, RangeOutMax=300000)
    B=0
    bPickA=Get bIsSkyDiving()
   )
   n6 = float * float(A=@n5, B=0.5)
   n7 = Get bIsParachuteOpened()
   > Set falling_anim_dyn_up_force_vec3(falling_anim_dyn_up_force_vec3:Vector=MakeVector(Y=0, Z=SelectFloat(A=@n6, B=@n5, bPickA=@n7))) -> (Output_Get:Vector)
   n8 = GetAnimInstance(self=GetOwnerSkeletalMeshComponent())
   > Set llama_head_phys_alpha(llama_head_phys_alpha:double=SelectFloat(
     A=0
     B=0.7
     bPickA=Montage_IsActive!(
      self:Anim Instance Object Reference=@n8
      Montage:Anim Montage Object Reference=/BRCosmetics/Animation/Game/MainPlayer/Emotes/Llama/Emote_llama_F
     ) -> (ReturnValue:bool)
    )) -> (Output_Get:double)
   Event BlueprintUpdateAnimation()
   Event OnCharacterCustomizationLoaded()
  ]}
]
variables [
 current_actor_z {type=number}
 current_dive_jiggle {type=Rotator}
 falling_anim_dyn_up_force_vec3 {type=Vector}
 isMannequin {type=boolean}
 llama_head_phys_alpha {type=number}
]
